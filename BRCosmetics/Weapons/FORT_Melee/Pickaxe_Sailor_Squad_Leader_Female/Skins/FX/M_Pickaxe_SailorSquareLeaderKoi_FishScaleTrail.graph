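# Unlit translucent Niagara ribbon material for a fish-scale trail on the Sailor Squad Leader Koi pickaxe.
# It pans swizzled, scaled UV1 over a fish-scale normal map, turning its blue channel into a hard scale pattern via SmoothStep and a cutoff.
# A soft SphereMask on UVs bent by Power and scaled in width shapes the pattern, which particle alpha then multiplies.
# Separate swipe and edge-mask nodes do not feed that chain.
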Material /BRCosmetics/Weapons/FORT_Melee/Pickaxe_Sailor_Squad_Leader_Female/Skins/FX/M_Pickaxe_SailorSquareLeaderKoi_FishScaleTrail
graph {
 properties {bUsedWithNiagaraRibbons=true, blendMode=BLEND_Translucent, pixelDepthOffsetMode=PDOM_Legacy, shadingModel=MSM_Unlit, usedShadingModels=Unlit}
 nodes [
  n1 = TextureCoordinate()
  AppendVector(A=OneMinus(None=ComponentMask(None=@n1) {r=true}), B=ComponentMask(None=@n1) {g=true})
  n2 = ScalarParameter() {defaultValue=0.85, parameterName=Scale Smooth Step, sliderMax=1}
  n3 = Panner(
   Coordinate=MaterialFunctionCall(XY=TextureCoordinate() {coordinateIndex=1}) {materialFunction=&Swizzle}.YX
   Speed=AppendVector(A=Constant(), B=ScalarParameter() {defaultValue=1, parameterName=Speed X, sliderMax=2})
  )
  n4 = Constant2Vector() {g=-1, r=1}
  n5 = ScalarParameter() {defaultValue=1, parameterName=Scale X, sliderMax=2}
  n6 = ScalarParameter() {defaultValue=1, parameterName=Scale Y, sliderMax=2}
  n7 = TextureSample(UVs=MaterialFunctionCall(UVs=Multiply(A=@n3, B=@n4), Texture Scale=AppendVector(A=@n5, B=@n6)) {materialFunction=&ScaleUVsByCenter}.UVs) {
   constCoordinate=1
   samplerType=SAMPLERTYPE_Normal
   texture=/Game/Characters/Player/Female/Medium/Bodies/F_MED_Lady_Atlantis/Textures/T_FishScales_N
  }
  n8 = If(
   A=OneMinus(None=SmoothStep(Min=@n2, Value=@n7.B) {constMin=0.85})
   B=ScalarParameter() {defaultValue=0.25, parameterName=Scale Cutoff, sliderMax=1}
   "A > B"=Constant() {r=1}
   "A < B"=Constant()
  ) {constB=0.25}
  n9 = TextureCoordinate()
  n10 = AppendVector(
   A=Power(Base=ComponentMask(None=@n9) {r=true}, Exp=ScalarParameter() {defaultValue=0.5, parameterName=Concave, sliderMax=5}) {constExponent=0.5}
   B=ComponentMask(None=@n9) {g=true}
  )
  n11 = Constant() {r=1}
  n12 = Multiply(A=ScalarParameter() {defaultValue=1.25, group=Candle Flame, parameterName=Candle Width, sliderMax=1})
  n13 = SphereMask(
   A=MaterialFunctionCall(UVs=@n10, Texture Scale=AppendVector(A=@n11, B=@n12)) {materialFunction=&ScaleUVsByCenter}.UVs
   B=Constant() {r=0.5}
  ) {attenuationRadius=0.35, hardnessPercent=0}
  Multiply(A=ParticleColor().A, B=Multiply(A=Multiply(A=@n8, B=@n13)))
  n14 = TextureCoordinate()
  n15 = TextureCoordinate()
  Multiply(
   A=SphereMask(A=ComponentMask(None=@n14) {g=true}, B=Constant() {r=0.5}) {attenuationRadius=0.5, hardnessPercent=50}
   B=SphereMask(A=ComponentMask(None=@n15) {r=true}, B=Constant() {r=0.4}) {attenuationRadius=0.4, hardnessPercent=0}
  )
  Multiply(A=@n1, B=@n10)
  ScalarParameter() {defaultValue=0.5, parameterName=Swipe}
  n16 = DynamicParameter() {defaultValue=<rgba 0.5 0.1 1 1>, paramNames [Swipe,Width,Param3,Param4]}
  SphereMask(A=ComponentMask(None=TextureCoordinate()) {r=true}, B=@n16.Swipe, Radius=@n16.Swipe) {attenuationRadius=0.1, hardnessPercent=0}
 ]
}
